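; BGA_Dancepad is a dance pad actor for the Mash playlist that marks player pawns standing inside its capsule.
; On capsule begin overlap, it casts to a Fort player pawn and applies GE_GrantOnDancepad to that pawn's ability system.
; On end overlap, it removes active effects tagged Athena.Location.Quest.Dance, so the state, likely used for dance quests, lasts only while on the pad.
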(event OnComponentBeginOverlap(Capsule)
  (params (OverlappedComponent :type Primitive%20Component%20Object%20Reference) (OtherActor :type Actor%20Object%20Reference) (OtherComp :type Primitive%20Component%20Object%20Reference) (OtherBodyIndex :type Integer) (bFromSweep :type Boolean) (SweepResult :type Hit%20Result%20Structure%20%28by%20ref%29))
  (bind _asfort_player_pawn (Utilities|Casting|CastToFortPlayerPawn OtherActor))
  (GameplayEffects|ApplyGameplayEffectToSelf :self (Ability|GetFortAbilitySystemComponent _asfort_player_pawn) :GameplayEffectClass "/Game/Athena/Playlists/Mash/Gameplay/GE_GrantOnDancepad.GE_GrantOnDancepad_C" :Level 0.0))

(event OnComponentEndOverlap(Capsule)
  (params (OverlappedComponent :type Primitive%20Component%20Object%20Reference) (OtherActor :type Actor%20Object%20Reference) (OtherComp :type Primitive%20Component%20Object%20Reference) (OtherBodyIndex :type Integer))
  (bind _asfort_player_pawn (Utilities|Casting|CastToFortPlayerPawn OtherActor))
  (GameplayEffects|RemoveActiveEffectswithTags (Ability|GetFortAbilitySystemComponent _asfort_player_pawn) "(GameplayTags=((TagName=\"Athena.Location.Quest.Dance\")))"))

(fn ConstructionScript)
